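# Control Rig for the Frozen Reality large male body that fits its deform arm bones and shoulder pads to the animated skeleton.
# Each arm partially follows the clavicle, copies upper and lower arm rotation, then uses PBIK to reach the hand.
# When skydiving, spherical pose readers on left upper arm and right lower arm weight shoulder pad offsets.
# A similar emote branch with interpolated weights has no execution input connected.
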ControlRigBlueprint /BRCosmetics/Characters/Player/Male/Large/Bodies/M_LRG_FrozenReality/Meshes/M_LRG_FrozenReality_CtrlRig
graphs [
 ContainedGraph {nodes [
   n1 = RigVMFunctionEntryNode()
   SequenceAggregate = &RigVMFunction_Sequence(ExecuteContext=@n1.ExecuteContext)
   SequenceAggregate_1 = &RigVMFunction_Sequence(ExecuteContext=@SequenceAggregate.B)
   SequenceAggregate_2 = &RigVMFunction_Sequence(ExecuteContext=@SequenceAggregate_1.B)
   RigVMFunctionReturnNode(A=@SequenceAggregate.A, B=@SequenceAggregate_1.A, C=@SequenceAggregate_2.A, D=@SequenceAggregate_2.B)
  ]}
 RigVMFunctionLibrary {links [], nodes []}
 RigVMModel {nodes [
   Branch_1_RigVMUnitNode = &RigVMFunction_ControlFlowBranch(ExecuteContext=(), Condition=RigVMVariableNode() {variable=Is_Emoting}.Value)
   n1 = &RigUnit_SphericalPoseReader(
    ActiveRegionScaleFactors=(PositiveWidth=4.5,NegativeWidth=1.5,PositiveHeight=4.5,NegativeHeight=4.5)
    ActiveRegionSize=0.1
    Debug=(bDrawDebug=false,bDrawLocalAxes=true,DebugScale=25,DebugSegments=20,DebugThickness=0.25)
    DriverAxis=(X=-1,Y=0,Z=0)
    DriverItem=(Type=Bone,Name="dfrm_upperarm_l")
    FalloffRegionScaleFactors=(PositiveWidth=1,NegativeWidth=1,PositiveHeight=1,NegativeHeight=1)
    FalloffSize=0.5
    FlipHeightScaling=false
    OptionalParentItem=(Type=None,Name="None")
    RotationOffset=(X=10,Y=25,Z=25)
    ExecutePin=@Branch_1_RigVMUnitNode.True
   )
   AlphaInterp = &RigVMFunction_AlphaInterp(InterpSpeedDecreasing=4, InterpSpeedIncreasing=10, Scale=1, bClampResult=False, Value=@n1.OutputParam)
   > &RigUnit_ModifyTransforms(
    ItemToModify=((Item=(Type=Bone,Name="dfrm_shoulderpad_parent_l"),Transform=(Rotation=(X=0,Y=0,Z=0,W=1),Translation=(X=12,Y=0,Z=0),Scale3D=(X=1,Y=1,Z=1))))
    Weight=@AlphaInterp.Result
    ItemToModify.0.Transform.Rotation=&RigVMFunction_MathQuaternionFromEuler(Euler=(X=0,Y=-20,Z=50)).Result
   )
   > n2 = &RigUnit_SphericalPoseReader(
    ActiveRegionScaleFactors=(PositiveWidth=4.5,NegativeWidth=1.5,PositiveHeight=4.5,NegativeHeight=4.5)
    ActiveRegionSize=0.1
    Debug=(bDrawDebug=false,bDrawLocalAxes=true,DebugScale=25,DebugSegments=20,DebugThickness=0.25)
    DriverAxis=(X=-1,Y=0,Z=0)
    DriverItem=(Type=Bone,Name="dfrm_lowerarm_r")
    FalloffRegionScaleFactors=(PositiveWidth=1,NegativeWidth=1,PositiveHeight=1,NegativeHeight=1)
    FalloffSize=0.5
    OptionalParentItem=(Type=Bone)
    RotationOffset=(X=10,Y=25,Z=25)
   )
   n3 = &RigVMFunction_AlphaInterp(InterpSpeedDecreasing=4, InterpSpeedIncreasing=6, Scale=1, bClampResult=true, bMapRange=false, Value=@n2.OutputParam)
   > &RigUnit_ModifyTransforms(
    ItemToModify=((Item=(Type=Bone,Name="dfrm_shoulderpad_parent_r"),Transform=(Rotation=(X=0,Y=0,Z=0,W=1),Translation=(X=-12,Y=0,Z=0),Scale3D=(X=1,Y=1,Z=1))))
    Weight=@n3.Result
    ItemToModify.0.Transform.Rotation=&RigVMFunction_MathQuaternionFromEuler(Euler=(X=0,Y=20,Z=50)).Result
   )
   RigUnit_BeginExecution = &RigUnit_BeginExecution()
   SequenceAggregate = RigVMAggregateNode(ExecuteContext=@RigUnit_BeginExecution.ExecutePin)
   Branch_RigVMUnitNode = &RigVMFunction_ControlFlowBranch(ExecuteContext=@SequenceAggregate.A, Condition=RigVMVariableNode() {variable=Is_Skydiving}.Value)
   n4 = &RigUnit_SphericalPoseReader(
    ActiveRegionScaleFactors=(PositiveWidth=4.5,NegativeWidth=1.5,PositiveHeight=4.5,NegativeHeight=4.5)
    ActiveRegionSize=0.1
    Debug=(bDrawDebug=false,bDrawLocalAxes=true,DebugScale=25,DebugSegments=20,DebugThickness=0.25)
    DriverAxis=(X=-1,Y=0,Z=0)
    DriverItem=(Type=Bone,Name="dfrm_upperarm_l")
    FalloffRegionScaleFactors=(PositiveWidth=1,NegativeWidth=1,PositiveHeight=1,NegativeHeight=1)
    FalloffSize=0.5
    FlipHeightScaling=false
    OptionalParentItem=(Type=None,Name="None")
    RotationOffset=(X=10,Y=25,Z=25)
    ExecutePin=@Branch_RigVMUnitNode.True
   )
   > &RigUnit_ModifyTransforms(
    ItemToModify=((Item=(Type=Bone,Name="dfrm_shoulderpad_parent_l"),Transform=(Rotation=(X=0,Y=0,Z=0,W=1),Translation=(X=16,Y=0,Z=-5),Scale3D=(X=1,Y=1,Z=1))))
    ItemToModify.0.Transform.Rotation=&RigVMFunction_MathQuaternionFromEuler(Euler=(X=0,Y=-20,Z=50)).Result
    Weight=@n4.OutputParam
   )
   > n5 = &RigUnit_SphericalPoseReader(
    ActiveRegionScaleFactors=(PositiveWidth=4.5,NegativeWidth=1.5,PositiveHeight=4.5,NegativeHeight=4.5)
    ActiveRegionSize=0.1
    Debug=(bDrawDebug=false,bDrawLocalAxes=true,DebugScale=25,DebugSegments=20,DebugThickness=0.25)
    DriverAxis=(X=-1,Y=0,Z=0)
    DriverItem=(Type=Bone,Name="dfrm_lowerarm_r")
    FalloffRegionScaleFactors=(PositiveWidth=1,NegativeWidth=1,PositiveHeight=1,NegativeHeight=1)
    FalloffSize=0.5
    OptionalParentItem=(Type=Bone)
    RotationOffset=(X=10,Y=25,Z=25)
   )
   > &RigUnit_ModifyTransforms(
    ItemToModify=((Item=(Type=Bone,Name="dfrm_shoulderpad_parent_r"),Transform=(Rotation=(X=0,Y=0,Z=0,W=1),Translation=(X=-15,Y=-5,Z=5),Scale3D=(X=1,Y=1,Z=1))))
    ItemToModify.0.Transform.Rotation=&RigVMFunction_MathQuaternionFromEuler(Euler=(X=0,Y=20,Z=50)).Result
    Weight=@n5.OutputParam
   )
   &RigUnit_RotationConstraintLocalSpaceOffset(
    AdvancedSettings=(RotationOrderForFilter=XZY)
    Child=(Type=Bone,Name="dfrm_clavicle_l")
    Filter=(bX=True,bY=True,bZ=True)
    Parents=((Item=(Type=Bone,Name="clavicle_l"),Weight=1))
    Weight=0.7
    ExecutePin=@SequenceAggregate.B
   )
   > &RigUnit_SetRotation(
    Item=(Type=Bone,Name="dfrm_upperarm_l")
    Space=GlobalSpace
    Weight=1
    bInitial=false
    bPropagateToChildren=true
    Value=&RigUnit_GetTransform(Item=(Type=Bone,Name="upperarm_l")).Transform.Rotation
   )
   > &RigUnit_SetRotation(
    Item=(Type=Bone,Name="dfrm_lowerarm_l")
    Space=GlobalSpace
    Weight=1
    bInitial=false
    bPropagateToChildren=true
    Value=&RigUnit_GetTransform(Item=(Type=Bone,Name="lowerarm_l")).Transform.Rotation
   )
   > PBIK = /Script/PBIK.RigUnit_PBIK(
    BoneSettings=((Bone="dfrm_lowerarm_l",X=Locked,Y=Locked,Z=Free,MinZ=0,MaxZ=0,PreferredAngles=(X=0,Y=0,Z=0)))
    Debug=(DrawScale=1)
    Effectors=((Bone="dfrm_hand_l",Transform=(Rotation=(X=0,Y=0,Z=0,W=1),Translation=(X=0,Y=0,Z=0),Scale3D=(X=1,Y=1,Z=1)),PositionAlpha=1,RotationAlpha=1,StrengthAlpha=1,PullChainAlpha=0,PinRotation=1,ChainDepth=0))
    ExcludedBones=()
    Root=dfrm_clavicle_l
    Settings="(Iterations=20,SubIterations=0,MassMultiplier=1,bAllowStretch=False,RootBehavior=PinToInput,PrePullRootSettings=(RotationAlpha=0,RotationAlphaX=1,RotationAlphaY=1,RotationAlphaZ=1,PositionAlpha=1,PositionAlphaX=1,PositionAlphaY=1,PositionAlphaZ=1),GlobalPu ...(truncated)"
    Effectors.0.Transform=&RigUnit_GetTransform(Item=(Type=Bone,Name="hand_l"), bInitial=false).Transform
   ) {is=struct RigVMUnitNode}
   &RigUnit_RotationConstraintLocalSpaceOffset(
    AdvancedSettings=(RotationOrderForFilter=XZY)
    Child=(Type=Bone,Name="dfrm_clavicle_r")
    Filter=(bX=true,bY=true,bZ=true)
    Parents=((Item=(Type=Bone,Name="clavicle_r"),Weight=1))
    Weight=0.75
    ExecutePin=@SequenceAggregate.C
   )
   > &RigUnit_SetRotation(
    Item=(Type=Bone,Name="dfrm_upperarm_r")
    Space=GlobalSpace
    Weight=1
    bInitial=False
    bPropagateToChildren=true
    Value=&RigUnit_GetTransform(Item=(Type=Bone,Name="upperarm_r")).Transform.Rotation
   )
   > &RigUnit_SetRotation(
    Item=(Type=Bone,Name="dfrm_lowerarm_r")
    Space=GlobalSpace
    Weight=1
    bInitial=False
    bPropagateToChildren=true
    Value=&RigUnit_GetTransform(Item=(Type=Bone,Name="lowerarm_r")).Transform.Rotation
   )
   > &RigUnit_PBIK(
    BoneSettings=((Bone="dfrm_lowerarm_r",RotationStiffness=0,X=Locked,Y=Locked,Z=Free,bUsePreferredAngles=false,PreferredAngles=(X=0,Y=0,Z=0)))
    Debug=(DrawScale=1,bDrawDebug=true)
    Effectors=((Bone="dfrm_hand_r",Transform=(Rotation=(X=0,Y=0,Z=0,W=1),Translation=(X=0,Y=0,Z=0),Scale3D=(X=1,Y=1,Z=1)),PositionAlpha=1,RotationAlpha=1,StrengthAlpha=1,PullChainAlpha=0,PinRotation=1,ChainDepth=0))
    ExcludedBones=()
    Root=dfrm_clavicle_r
    Settings="(Iterations=20,SubIterations=0,MassMultiplier=1,RootBehavior=PinToInput,PrePullRootSettings=(RotationAlphaX=1,RotationAlphaY=1,RotationAlphaZ=1,PositionAlpha=1,PositionAlphaX=1,PositionAlphaY=1,PositionAlphaZ=1),GlobalPullChainAlpha=1,MaxAngle=30,OverRelax ...(truncated)"
    Effectors.0.Transform=&RigUnit_GetTransform(Item=(Type=Bone,Name="hand_r")).Transform
   )
   > &RigUnit_ParentConstraint(
    AdvancedSettings=(RotationOrderForFilter=XZY)
    Child=(Type=Bone,Name="weapon_lowerarm_r")
    Parents=((Item=(Type=Bone,Name="dfrm_lowerarm_r"),Weight=1))
    Weight=1
   )
  ]}
]
hierarchy [
 root {loc=<0 0 0>, rot=<0 0 0 1>}
 pelvis {parent=root, loc=<0 -0.7321 92.4218>, rot=<0.0127 0.707 -0.0127 -0.707>}
 spine_01 {parent=pelvis, loc=<5.6002 -0.5128 0>, rot=<0 0 0.034 -0.9994>}
 spine_02 {parent=spine_01, loc=<8.0753 0.0139 0>, rot=<0 0 0.0238 -0.9997>}
 spine_03 {parent=spine_02, loc=<7.0393 0 0>, rot=<0 0 0.0149 -0.9999>}
 spine_04 {parent=spine_03, loc=<8.1024 0 0>, rot=<0 0 -0.0132 -0.9999>}
 spine_05 {parent=spine_04, loc=<10.0146 0 0>, rot=<0 0 -0.0015 -1>}
 clavicle_l {parent=spine_05, loc=<2.5856 -0.2866 -4.3777>, rot=<0.0465 -0.7289 -0.009 -0.6829>}
 upperarm_l {parent=clavicle_l, loc=<18.2264 0.1342 -0.001>, rot=<0.0253 -0.2693 0.0013 -0.9627>}
 lowerarm_l {parent=upperarm_l, loc=<29.0071 0.1686 -1.239>, rot=<0.0391 -0.0099 -0.2453 -0.9686>}
 hand_l {parent=lowerarm_l, loc=<25.8607 0.222 1.0274>, rot=<-0.7193 -0.0988 0.1193 -0.6773>}
 thumb_01_l {parent=hand_l, loc=<1.3118 -0.3191 -2.459>, rot=<0.6546 -0.4042 0.0664 -0.6354>}
 thumb_02_l {parent=thumb_01_l, loc=<5.9336 -0.0448 0.0401>, rot=<-0.0037 0.0213 0.298 -0.9543>}
 thumb_03_l {parent=thumb_02_l, loc=<2.2223 0.3682 -0.0907>, rot=<0.0118 -0.013 0.0066 -0.9998>}
 index_metacarpal_l {parent=hand_l, loc=<1.6535 0.0183 -2.96>, rot=<0.0158 -0.0124 -0.0173 -0.9996>}
 index_01_l {parent=index_metacarpal_l, loc=<8.3008 0.0357 -0.0524>, rot=<-0.0044 0.0485 0.0924 -0.9945>}
 index_02_l {parent=index_01_l, loc=<3.3587 -0.3004 0.0452>, rot=<-0.002 0.0027 0.3225 -0.9466>}
 index_03_l {parent=index_02_l, loc=<2.5578 -0.123 -0.0252>, rot=<-0.002 0.0067 0.2558 -0.9667>}
 middle_metacarpal_l {parent=hand_l, loc=<2.1856 -0.1725 -0.5984>, rot=<0.006 0.0168 -0.0329 -0.9993>}
 middle_01_l {parent=middle_metacarpal_l, loc=<7.3275 -0.2447 0.0129>, rot=<-0.0208 0.0042 0.0858 -0.9961>}
 middle_02_l {parent=middle_01_l, loc=<4.0123 0.1401 0.1175>, rot=<-0.0218 0.0258 0.3512 -0.9357>}
 middle_03_l {parent=middle_02_l, loc=<2.6077 0.0534 0.1373>, rot=<0.0082 0.0188 0.1708 -0.9851>}
 ring_metacarpal_l {parent=hand_l, loc=<2.2074 -0.3625 1.3638>, rot=<-0.0284 0.0722 -0.0458 -0.9959>}
 ring_01_l {parent=ring_metacarpal_l, loc=<7.0295 -0.6197 -0.1446>, rot=<-0.0489 -0.0252 0.1892 -0.9804>}
 ring_02_l {parent=ring_01_l, loc=<3.6174 0.4647 -0.1027>, rot=<0 0 0.2806 -0.9598>}
 ring_03_l {parent=ring_02_l, loc=<2.4152 0.114 0.0199>, rot=<0 0 0.1189 -0.9929>}
 pinky_metacarpal_l {parent=hand_l, loc=<1.7616 -0.9753 3.3814>, rot=<-0.0397 0.0814 -0.069 -0.9935>}
 pinky_01_l {parent=pinky_metacarpal_l, loc=<7.1709 -0.4831 -0.1094>, rot=<-0.0654 -0.0035 0.2242 -0.9723>}
 pinky_02_l {parent=pinky_01_l, loc=<2.8274 0.0893 -0.1043>, rot=<0.0037 0.0011 0.2854 -0.9584>}
 pinky_03_l {parent=pinky_02_l, loc=<2.2046 0.0307 -0.0078>, rot=<-0.0002 0.0098 0.1445 -0.9895>}
 lowerarm_twist_02_l {parent=lowerarm_l, loc=<9.6199 0 0>, rot=<0 0 0 -1>}
 lowerarm_twist_01_l {parent=lowerarm_l, loc=<18.0217 -0.2539 2.0925>, rot=<0 0 0 -1>}
 upperarm_twist_01_l {parent=upperarm_l, loc=<7.9796 0.0843 -0.6195>, rot=<0 0 0 -1>}
 upperarm_twist_02_l {parent=upperarm_l, loc=<19.5059 0.0843 -0.6195>, rot=<0 0 0 -1>}
 attach_cape {parent=spine_05, loc=<9.1509 -13.3783 0>, rot=<0.0616 -0.7044 -0.0616 -0.7044>}
 attach_backpack {parent=spine_05, loc=<-0.123 -15.1423 0>, rot=<-0.0093 -0.707 0.0093 -0.707>}
 neck_01 {parent=spine_05, loc=<10.815 1.9196 0>, rot=<0 0 -0.1864 -0.9825>}
 neck_02 {parent=neck_01, loc=<4.7347 0.3045 0>, rot=<0 0 0.0363 -0.9993>}
 head {parent=neck_02, loc=<5.7632 0.5552 0>, rot=<0 0 0.115 -0.9934>}
 clavicle_r {parent=spine_05, loc=<2.5855 -0.2866 4.3777>, rot=<0.7289 0.0465 -0.6829 0.009>}
 upperarm_r {parent=clavicle_r, loc=<-18.2264 -0.1342 0.001>, rot=<0.0253 -0.2693 0.0013 -0.9627>}
 lowerarm_r {parent=upperarm_r, loc=<-29.0071 -0.1686 1.239>, rot=<0.0391 -0.0099 -0.2453 -0.9686>}
 hand_r {parent=lowerarm_r, loc=<-25.8607 -0.222 -1.0274>, rot=<-0.7193 -0.0988 0.1193 -0.6773>}
 thumb_01_r {parent=hand_r, loc=<-1.3118 0.3191 2.459>, rot=<0.6546 -0.4042 0.0664 -0.6354>}
 thumb_02_r {parent=thumb_01_r, loc=<-5.9336 0.0448 -0.0401>, rot=<-0.0037 0.0213 0.298 -0.9543>}
 thumb_03_r {parent=thumb_02_r, loc=<-2.2223 -0.3682 0.0907>, rot=<0.0118 -0.013 0.0066 -0.9998>}
 index_metacarpal_r {parent=hand_r, loc=<-1.6535 -0.0183 2.96>, rot=<0.0158 -0.0124 -0.0173 -0.9996>}
 index_01_r {parent=index_metacarpal_r, loc=<-8.3008 -0.0357 0.0524>, rot=<-0.0044 0.0485 0.0924 -0.9945>}
 index_02_r {parent=index_01_r, loc=<-3.3587 0.3004 -0.0452>, rot=<-0.002 0.0027 0.3225 -0.9466>}
 index_03_r {parent=index_02_r, loc=<-2.5578 0.123 0.0252>, rot=<-0.002 0.0067 0.2558 -0.9667>}
 middle_metacarpal_r {parent=hand_r, loc=<-2.1856 0.1725 0.5984>, rot=<0.006 0.0168 -0.0329 -0.9993>}
 middle_01_r {parent=middle_metacarpal_r, loc=<-7.3275 0.2447 -0.0129>, rot=<-0.0208 0.0042 0.0858 -0.9961>}
 middle_02_r {parent=middle_01_r, loc=<-4.0123 -0.1401 -0.1175>, rot=<-0.0218 0.0258 0.3512 -0.9357>}
 middle_03_r {parent=middle_02_r, loc=<-2.6077 -0.0534 -0.1373>, rot=<0.0082 0.0188 0.1708 -0.9851>}
 ring_metacarpal_r {parent=hand_r, loc=<-2.2074 0.3625 -1.3638>, rot=<-0.0284 0.0722 -0.0458 -0.9959>}
 ring_01_r {parent=ring_metacarpal_r, loc=<-7.0295 0.6197 0.1446>, rot=<-0.0489 -0.0252 0.1892 -0.9804>}
 ring_02_r {parent=ring_01_r, loc=<-3.6174 -0.4647 0.1027>, rot=<0 0 0.2806 -0.9598>}
 ring_03_r {parent=ring_02_r, loc=<-2.4152 -0.114 -0.0199>, rot=<0 0 0.1189 -0.9929>}
 pinky_metacarpal_r {parent=hand_r, loc=<-1.7616 0.9753 -3.3814>, rot=<-0.0397 0.0814 -0.069 -0.9935>}
 pinky_01_r {parent=pinky_metacarpal_r, loc=<-7.1709 0.4831 0.1094>, rot=<-0.0654 -0.0035 0.2242 -0.9723>}
 pinky_02_r {parent=pinky_01_r, loc=<-2.8274 -0.0893 0.1043>, rot=<0.0037 0.0011 0.2854 -0.9584>}
 pinky_03_r {parent=pinky_02_r, loc=<-2.2046 -0.0307 0.0078>, rot=<-0.0002 0.0098 0.1445 -0.9895>}
 lowerarm_twist_02_r {parent=lowerarm_r, loc=<-9.6199 0 0>, rot=<0 0 0 -1>}
 lowerarm_twist_01_r {parent=lowerarm_r, loc=<-17.8549 0 0>, rot=<0 0 0 -1>}
 weapon_lowerarm_r {parent=lowerarm_r, loc=<0 0 0>, rot=<0 0 0 1>}
 upperarm_twist_01_r {parent=upperarm_r, loc=<-7.9796 -0.0843 0.6195>, rot=<0 0 0 -1>}
 upperarm_twist_02_r {parent=upperarm_r, loc=<-19.5059 -0.0843 0.6195>, rot=<0 0 0 -1>}
 dfrm_clavicle_l {parent=spine_05, loc=<2.5856 -0.2866 -4.3777>, rot=<0.0465 -0.7289 -0.009 -0.6829>}
 dfrm_upperarm_l {parent=dfrm_clavicle_l, loc=<18.2264 0.1342 -0.001>, rot=<0.0136 -0.2686 0.0045 -0.9631>}
 dfrm_lowerarm_l {parent=dfrm_upperarm_l, loc=<28.9168 0.3937 -3.0758>, rot=<0.0278 -0.0102 -0.2132 -0.9766>}
 dfrm_hand_l {parent=dfrm_lowerarm_l, loc=<25.7169 1.5907 2.3779>, rot=<-0.6996 -0.1246 0.1045 -0.6958>}
 dfrm_shoulderpad_base_l {parent=dfrm_upperarm_l, loc=<-0.5526 -0.2067 7.9461>, rot=<0.9993 0.0285 -0.0257 0.0033>}
 dfrm_shoulderpad_parent_l {parent=dfrm_shoulderpad_base_l, loc=<0 0 0>, rot=<0 0 0 1>}
 const_brace_outer_base_l {parent=dfrm_clavicle_l, loc=<8.1917 7.5493 9.4749>, rot=<-0.5105 -0.799 0.3049 -0.0897>}
 const_brace_outer_parent_l {parent=const_brace_outer_base_l, loc=<0 0 0>, rot=<0 0 0 1>}
 const_brace_inner_base_l {parent=dfrm_clavicle_l, loc=<6.8121 7.7071 9.65>, rot=<-0.5105 -0.799 0.3049 -0.0897>}
 const_brace_inner_parent_l {parent=const_brace_inner_base_l, loc=<0 0 0>, rot=<0 0 0 1>}
 dfrm_clavicle_r {parent=spine_05, loc=<2.5855 -0.2866 4.3777>, rot=<0.7289 0.0465 -0.6829 0.009>}
 dfrm_upperarm_r {parent=dfrm_clavicle_r, loc=<-18.2264 -0.1342 0.0005>, rot=<0.0136 -0.2686 0.0045 -0.9631>}
 dfrm_lowerarm_r {parent=dfrm_upperarm_r, loc=<-28.9172 -0.3938 3.0764>, rot=<0.0278 -0.0102 -0.2132 -0.9766>}
 dfrm_hand_r {parent=dfrm_lowerarm_r, loc=<-25.7168 -1.5907 -2.3781>, rot=<-0.6996 -0.1246 0.1045 -0.6958>}
 dfrm_shoulderpad_base_r {parent=dfrm_upperarm_r, loc=<0.5526 0.2067 -7.9461>, rot=<0.9993 0.0285 -0.0257 0.0033>}
 dfrm_shoulderpad_parent_r {parent=dfrm_shoulderpad_base_r, loc=<0 0 0>, rot=<0 0 0 1>}
 const_brace_outer_base_r {parent=dfrm_clavicle_r, loc=<-8.1917 -7.5493 -9.4749>, rot=<0.3049 0.0897 0.5105 -0.799>}
 const_brace_outer_parent_r {parent=const_brace_outer_base_r, loc=<0 0 0>, rot=<0 0 0 1>}
 const_brace_inner_base_r {parent=dfrm_clavicle_r, loc=<-6.8121 -7.7071 -9.65>, rot=<0.3049 0.0897 0.5105 -0.799>}
 const_brace_inner_parent_r {parent=const_brace_inner_base_r, loc=<0 0 0>, rot=<0 0 0 1>}
 dfrm_neck_01 {parent=spine_05, loc=<12.3457 -0.3291 -0.0855>, rot=<0 0 -0.1864 -0.9825>}
 dfrm_neck_02 {parent=dfrm_neck_01, loc=<5.107 -0.9415 0.0855>, rot=<0 0 -0.0363 -0.9993>}
 dfrm_head {parent=dfrm_neck_02, loc=<6.091 -0.6202 0>, rot=<0 0 0.151 -0.9885>}
 thigh_l {parent=pelvis, loc=<-1.8716 -0.0109 -11.9936>, rot=<0.0012 0.0572 0.0155 -0.9982>}
 calf_l {parent=thigh_l, loc=<-41.9178 -0.2246 0.0926>, rot=<0.0006 -0.0161 -0.0391 -0.9991>}
 foot_l {parent=calf_l, loc=<-38.7057 1.2438 0.2452>, rot=<0.0975 -0.0474 0.0317 -0.9936>}
 ball_l {parent=foot_l, loc=<-6.1614 13.1469 0.6807>, rot=<0.0001 -0.0001 -0.7063 -0.7079>}
 calf_twist_02_l {parent=calf_l, loc=<-15.9268 -0.0967 0.0397>, rot=<0.0237 0.0095 -0.0035 -0.9997>}
 thigh_twist_01_l {parent=thigh_l, loc=<-21.6303 0 0>, rot=<0.0013 0 0 -1>}
 const_thigh_l {parent=thigh_l, loc=<0 0 0>, rot=<0 0 0 1>}
 thigh_r {parent=pelvis, loc=<-1.8716 -0.0109 11.9936>, rot=<0.0572 -0.0012 0.9982 0.0155>}
 calf_r {parent=thigh_r, loc=<41.9178 0.2246 -0.0926>, rot=<0.0006 -0.0161 -0.0391 -0.9991>}
 foot_r {parent=calf_r, loc=<38.7057 -1.2438 -0.2452>, rot=<0.0975 -0.0474 0.0317 -0.9936>}
 ball_r {parent=foot_r, loc=<6.1614 -13.1469 -0.6807>, rot=<0.0001 -0.0001 -0.7063 -0.7079>}
 calf_twist_02_r {parent=calf_r, loc=<15.9268 0.0967 -0.0397>, rot=<0.0237 0.0095 -0.0035 -0.9997>}
 thigh_twist_01_r {parent=thigh_r, loc=<21.6303 0 0>, rot=<0.0013 0 0 -1>}
 const_thigh_r {parent=thigh_r, loc=<0 0 0>, rot=<0 0 0 1>}
]
variables [
 Is_Emoting {type=bool, default=False}
 Is_Skydiving {type=bool, default=False}
]
