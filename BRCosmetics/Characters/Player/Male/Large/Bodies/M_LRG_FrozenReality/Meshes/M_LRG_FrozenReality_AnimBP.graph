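# Animation Blueprint for the Frozen Reality large male body cosmetic, layering secondary motion onto the pose copied from the attached parent.
# A Control Rig feeds thigh and left-brace corrections, driven by limb twist deltas, then weighted brace constraints.
# In the front end, an idle curve blends slot animation onto the shoulder pads.
# Event logic toggles the "Hide Element 01" material parameter using built-in emote, pickaxe tag, skydiving, sprint and emote states.
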AnimBlueprint /BRCosmetics/Characters/Player/Male/Large/Bodies/M_LRG_FrozenReality/Meshes/M_LRG_FrozenReality_AnimBP
target_skeleton=&/M_LRG_FrozenReality_Skeleton
graphs [
 AnimGraph {nodes [
   n1 = Layered blend per bone(
    BasePose=Use cached pose 'StartPose'()
    BlendPoses_0=Slot 'DefaultSlot'()
    BlendWeights_0=Get PlayingFrontendIdle() -> (PlayingFrontendIdle:double)
   ) {layerSetup [{branchFilters [{boneName=dfrm_shoulderpad_base_l}, {boneName=dfrm_shoulderpad_base_r}]}]}
   Output Pose(Result=@n1) {layerGroup=None}
   n2 = Control Rig!(
    Source=Copy Pose From Mesh() {bUseAttachedParent=true}
    Alpha
    Is_Skydiving:bool=Get is_skydiving() -> (is_skydiving:bool)
    Is_Emoting:bool=Get bIsPlayingEmote()
   ) -> (Pose) {
    alphaBoolBlend {blendOption=Linear}
    controlRigAssetReference {blueprintRigClass=&/M_LRG_FrozenReality_CtrlRig.M_LRG_FrozenReality_CtrlRig_C}
   }
   n3 = "Transform (Modify) Bone - Bone: const_thigh_l"(
    ComponentPose=Local To Component(LocalPose=@n2)
    Alpha=Get legSwingDelta_l() -> (legSwingDelta_l:double)
    Translation="0,0,0"
    Rotation="0,0,0"
    Scale="1,1.5,1.5"
   ) {
    scaleMode=BMM_Additive
    scaleSpace=BCS_BoneSpace
    alphaBoolBlend {blendOption=Linear}
    alphaScaleBiasClamp {interpSpeedDecreasing=6, interpSpeedIncreasing=4}
    boneToModify {boneName=const_thigh_l}
    scale {y=1.5, z=1.5}
   }
   n4 = "Transform (Modify) Bone - Bone: const_thigh_r"(ComponentPose=@n3, Alpha=Get legSwingDelta_r() -> (legSwingDelta_r:double), Translation="0,0,0", Rotation="0,0,0", Scale="1.2,2,2") {
    scaleMode=BMM_Additive
    scaleSpace=BCS_BoneSpace
    alphaBoolBlend {blendOption=Linear}
    alphaScaleBiasClamp {interpSpeedDecreasing=6, interpSpeedIncreasing=4}
    boneToModify {boneName=const_thigh_r}
    scale=<1.2 2 2>
   }
   n5 = Get armSwingDelta_l() -> (armSwingDelta_l:double)
   n6 = "Transform (Modify) Bone - Bone: const_brace_outer_parent_l"(ComponentPose=@n4, Alpha=@n5, Translation="0,0.25,-0.5", Rotation="10.369,-0.094,0.064", Scale="1,1,1") {
    rotationMode=BMM_Additive
    rotationSpace=BCS_BoneSpace
    translationMode=BMM_Additive
    translationSpace=BCS_BoneSpace
    alphaBoolBlend {blendOption=Linear}
    boneToModify {boneName=const_brace_outer_parent_l}
    rotation=<rot 10.369 -0.094 0.064>
    translation {y=0.25, z=-0.5}
   }
   n7 = "Transform (Modify) Bone - Bone: const_brace_inner_parent_l"(ComponentPose=@n6, Alpha=@n5, Translation="0,0,-0.5", Rotation="12.157,-0.132,0.039", Scale="1,1,1") {
    rotationMode=BMM_Additive
    rotationSpace=BCS_BoneSpace
    translationMode=BMM_Additive
    translationSpace=BCS_BoneSpace
    alphaBoolBlend {blendOption=Linear}
    boneToModify {boneName=const_brace_inner_parent_l}
    rotation=<rot 12.157 -0.132 0.039>
    translation {z=-0.5}
   }
   n8 = "Constraint - const_brace_outer_base_r (PPPP)"!(
    ComponentPose=@n7
    Alpha=1
    ConstraintWeights_0=0.0418
    ConstraintWeights_1:float=0.4141
    ConstraintWeights_2:float=0.0253
    ConstraintWeights_3:float=0.5188
   ) -> (Pose) {
    alphaBoolBlend {blendOption=Linear}
    boneToModify {boneName=const_brace_outer_base_r}
    constraintSetup [
     {targetBone {boneName=spine_04}}
     {targetBone {boneName=spine_05}}
     {targetBone {boneName=dfrm_clavicle_l}}
     {targetBone {boneName=dfrm_clavicle_r}}
    ]
    constraintWeights [1,1,1,1]
   }
   n9 = "Constraint - const_brace_inner_base_r (PPPP)"!(
    ComponentPose=@n8
    Alpha=1
    ConstraintWeights_0=0.042
    ConstraintWeights_1:float=0.4179
    ConstraintWeights_2:float=0.0258
    ConstraintWeights_3:float=0.5143
   ) -> (Pose) {
    alphaBoolBlend {blendOption=Linear}
    boneToModify {boneName=const_brace_inner_base_r}
    constraintSetup [
     {targetBone {boneName=spine_04}}
     {targetBone {boneName=spine_05}}
     {targetBone {boneName=dfrm_clavicle_l}}
     {targetBone {boneName=dfrm_clavicle_r}}
    ]
    constraintWeights [1,1,1,1]
   }
   n10 = "Constraint - const_brace_inner_base_l (PPPPPPPP)"!(
    ComponentPose=@n9
    Alpha=1
    ConstraintWeights_0=0.0004
    ConstraintWeights_1:float=0.0002
    ConstraintWeights_2:float=0.0015
    ConstraintWeights_3:float=0.012
    ConstraintWeights_4:float=0.0482
    ConstraintWeights_5:float=0.4144
    ConstraintWeights_6:float=0.4775
    ConstraintWeights_7:float=0.046
   ) -> (Pose) {
    alphaBoolBlend {blendOption=Linear}
    boneToModify {boneName=const_brace_inner_base_l}
    constraintSetup [
     {targetBone {boneName=pelvis}}
     {targetBone {boneName=spine_01}}
     {targetBone {boneName=spine_02}}
     {targetBone {boneName=spine_03}}
     {targetBone {boneName=spine_04}}
     {targetBone {boneName=spine_05}}
     {targetBone {boneName=dfrm_clavicle_l}}
     {targetBone {boneName=dfrm_clavicle_r}}
    ]
    constraintWeights [1,1,1,1,1,1,1,1]
   }
   n11 = "Constraint - const_brace_outer_base_l (PPPPPPPP)"!(
    ComponentPose=@n10
    Alpha=1
    ConstraintWeights_0=0.0005
    ConstraintWeights_1:float=0.0006
    ConstraintWeights_2:float=0.0041
    ConstraintWeights_3:float=0.0273
    ConstraintWeights_4:float=0.0505
    ConstraintWeights_5:float=0.3697
    ConstraintWeights_6:float=0.5111
    ConstraintWeights_7:float=0.0362
   ) -> (Pose) {
    alphaBoolBlend {blendOption=Linear}
    boneToModify {boneName=const_brace_outer_base_l}
    constraintSetup [
     {targetBone {boneName=pelvis}}
     {targetBone {boneName=spine_01}}
     {targetBone {boneName=spine_02}}
     {targetBone {boneName=spine_03}}
     {targetBone {boneName=spine_04}}
     {targetBone {boneName=spine_05}}
     {targetBone {boneName=dfrm_clavicle_l}}
     {targetBone {boneName=dfrm_clavicle_r}}
    ]
    constraintWeights [1,1,1,1,1,1,1,1]
   }
   Save cached pose 'StartPose'(Pose=Component To Local(ComponentPose=@n11)) {cacheName=StartPose}
  ]}
 EventGraph {nodes [
   Event BlueprintUpdateAnimation()
   > n1 = Sequence!(execute) -> (then_0, then_1, then_2:Exec, then_3:Exec, then_4:Exec)
   n2 = Branch(execute=@n1.then_4, Condition=Get bIsInFrontEnd())
   > n3 = Sequence!(execute) -> (then_0, then_1)
   > Branch(Condition=Get bIsInFrontEnd())
   > Set PlayingFrontendIdle(PlayingFrontendIdle:double=GetCurveValue(self=GetOwnerFrontEndAnimInstance(), CurveName=FrontEnd_Idle_Bypass)) -> (Output_Get:double)
   Branch(execute=@n3.then_1, Condition=Get Is Builtin Emote() -> (Is Builtin Emote:bool))
   > n4 = Get Dynamic Material Instance(SkeletalMeshComponent=GetPartSkeletalMeshComponent())
   > SetScalarParameterValue(self=@n4, ParameterName=Hide Element 01, Value=1)
   > Delay(Duration=9)
   > SetScalarParameterValue(self=@n4, ParameterName=Hide Element 01, Value=0)
   n5 = Get Dynamic Material Instance(execute=[@n14.then, @n2.else], SkeletalMeshComponent=GetPartSkeletalMeshComponent())
   n6 = PickaxeItemDefinitionHasAnyMetaTag(
    Tags=Make Array!("[0]"=(TagName="Cosmetics.MetaTags.FrozenReality"), "[1]":Gameplay Tag Structure=(TagName="Cosmetics.Pickaxe.HighMotion")) -> (Array)
    bRequiresEquipped=true
   )
   > n7 = Branch(Condition=OR Boolean(A=Get Is Builtin Emote(), B=@n6))
   > n8 = Branch(Condition=OR Boolean(A=Get bIsSkyDiving(), B=Get bIsTacticalSprinting()))
   SetScalarParameterValue(execute=[@n9.else, @n15.then, @n16.else, @n8.then, @n7.else], self=@n5, ParameterName=Hide Element 01, Value=0)
   n9 = Branch(execute=@n8.else, Condition=Get bIsPlayingEmote())
   SetScalarParameterValue(execute=[@n16.then, @n9.then], self=@n5, ParameterName=Hide Element 01, Value=1)
   Set is_skydiving(
    execute=@n1.then_0
    is_skydiving:bool=OR Boolean!(A=Get bIsSkyDiving(), B=Get bIsSkydiveDiveMode(), C:bool=Get bIsSkydiveFloating()) -> (ReturnValue)
   ) -> (Output_Get:bool)
   n10 = Get Skel_Mesh()
   Set armSwingDelta_l(
    execute=@n1.then_2
    armSwingDelta_l:double=GetAlphaFromReferenceTransformDelta(
     SkeletalMeshComponent=@n10
     BoneName=clavicle_l
     BaseName=upperarm_l
     DeltaComponentType=QuaternionTwist
     InputRange=(X=-2,Y=34)
     TwistAxis="0,0,1"
    )
   ) -> (Output_Get:double)
   Set legSwingDelta_r(
    execute=@n1.then_1
    legSwingDelta_r:double=GetAlphaFromReferenceTransformDelta(
     SkeletalMeshComponent=@n10
     BoneName=thigh_r
     BaseName=pelvis
     DeltaComponentType=QuaternionTwist
     InputRange=(X=-27,Y=108)
     TwistAxis="0,0,1"
    )
   ) -> (Output_Get:double)
   > Set legSwingDelta_l(legSwingDelta_l:double=GetAlphaFromReferenceTransformDelta(
     SkeletalMeshComponent=@n10
     BoneName=thigh_l
     BaseName=pelvis
     DeltaComponentType=QuaternionTwist
     InputRange=(X=-29,Y=107)
     TwistAxis="0,0,1"
    )) -> (Output_Get:double)
   n11 = Branch(execute=@n1.then_3, Condition=Get bIsInFrontEnd())
   > n12 = Set BuiltIn Emote(BuiltIn Emote:float=GetCurveValue(self=GetOwnerFrontEndAnimInstance(), CurveName=emote_BuiltIn)) -> (Output_Get:float)
   > Set Is Builtin Emote(Is Builtin Emote:bool="To Boolean (Integer)"!(InInt:int=Truncate(A=@n12)) -> (ReturnValue:bool)) -> (Output_Get:bool)
   n13 = Set BuiltIn Emote(
    execute=@n11.else
    BuiltIn Emote=GetCurveValue(
     self=GetOwnerFortPlayerAnimInstance!(self:Custom Character Part Anim Instance Object Reference) -> (ReturnValue:Fort Player Anim Instance Object Reference)
     CurveName=emote_BuiltIn
    )
   )
   > Set Is Builtin Emote(Is Builtin Emote="To Boolean (Integer)"(InInt=Truncate(A=@n13)))
   n14 = Event OnWeaponChanged()
   n15 = Event OnBeginSkydiving()
   Event OnEndSkydiving()
   > n16 = Branch(Condition=@n6)
   Event BlueprintInitializeAnimation()
   > Set Skel_Mesh(Skel_Mesh=GetOwnerSkeletalMeshComponent())
  ]}
]
variables [
 Anim_Inst {type=AnimInstance, default=null}
 armSwingDelta_l {type=number}
 BuiltIn Emote {type=number}
 Is Builtin Emote {type=boolean}
 is_skydiving {type=boolean}
 legSwingDelta_l {type=number}
 legSwingDelta_r {type=number}
 PlayingFrontendIdle {type=number}
 ShoulderPosition {type=number}
 Skel_Mesh {type=SkeletalMeshComponent}
]
